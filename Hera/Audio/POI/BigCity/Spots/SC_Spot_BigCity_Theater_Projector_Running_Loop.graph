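# Looping ambient spot Sound Cue for a running film projector in the Big City POI theater, classed as indoor ambience.
# A looping wave passes through a default Modulator into an endlessly looping Enveloper that gently dips and restores volume over each cycle.
# Overridden attenuation adds occlusion with low-pass filtering, listener focus and reverb sends, and concurrency is tightly capped.
# It also feeds the exterior ambient source bus despite its indoor sound class.
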SoundCue /Hera/Audio/POI/BigCity/Spots/SC_Spot_BigCity_Theater_Projector_Running_Loop
root=Enveloper_0
nodes [
 {
  id=Enveloper_0
  children [Modulator_0]
  properties {
   bLoop=true
   bLoopIndefinitely=true
   loopEnd=20
   pitchCurve {editorCurveData {keys [{time=0}]}}
   volumeCurve {editorCurveData {keys [
      {arriveTangent=-0.01, interpMode=RCIM_Cubic, time=0, value=1}
      <at 10 0.7 RCIM_Cubic>
      {interpMode=RCIM_Cubic, leaveTangent=0.01, time=20, value=1}
     ]}}
  }
 }
 {id=Modulator_0, children [WavePlayer_0]}
 {id=WavePlayer_0, properties {bLooping=true, soundWaveAssetPtr=&/Spot_BigCity_Theater_Projector_Running_Loop}}
]
properties {
 bOverrideAttenuation=true
 bOverrideConcurrency=true
 maxDistance=1485
 soundClassObject=&Fort_Ambient_Indoors_Athena
 volumeMultiplier=0.5
 attenuationOverrides {
  bEnableListenerFocus=true
  bEnableOcclusion=true
  bEnableSubmixSends=true
  dBAttenuationAtMax=-40
  falloffDistance=1350
  focusVolumeAttenuation=1.1
  nonFocusVolumeAttenuation=0.6
  occlusionInterpolationTime=1
  occlusionLowPassFilterFrequency=750
  occlusionTraceChannel=ECC_GameTraceChannel11
  occlusionVolumeAttenuation=0.5
  reverbWetLevelMin=0.6905
  stereoSpread=100
  attenuationShapeExtents=<135 161.3869>
 }
 busSends [{audioBus=null, minSendLevel=1, sendLevel=1, soundSourceBus=&AmbientAudio_Exterior_SourceBus}]
 concurrencyOverrides {maxCount=2, voiceStealReleaseTime=1}
 soundSubmixSends [{
   minSendLevel=0.2
   soundSubmix=&Fortnite_Reverb_Default
   customSendLevelCurve {editorCurveData {keys [<at 0 0.4>,<at 1 1>]}}
  }]
}
