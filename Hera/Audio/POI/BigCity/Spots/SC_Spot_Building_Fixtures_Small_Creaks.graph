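# Sound Cue for a Big City POI ambient spot one-shot of small building fixture creaks, in the indoor ambient sound class.
# A Random node picks among twelve equally weighted creak waves, and a Modulator slightly varies the volume of each play.
# Falloff comes from a shared wider one-shot tile-ambient attenuation asset.
# A concurrency override caps the cue at two simultaneous instances, and its debug flag is left on.
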SoundCue /Hera/Audio/POI/BigCity/Spots/SC_Spot_Building_Fixtures_Small_Creaks
root=Modulator_0
nodes [
 {id=Modulator_0, children [Random_2], properties {volumeMax=1, volumeMin=0.9}}
 {
  id=Random_2
  children [
   WavePlayer_24
   WavePlayer_25
   WavePlayer_26
   WavePlayer_27
   WavePlayer_28
   WavePlayer_29
   WavePlayer_30
   WavePlayer_31
   WavePlayer_32
   WavePlayer_33
   WavePlayer_34
   WavePlayer_35
  ]
  properties {weights [1,1,1,1,1,1,1,1,1,1,1,1]}
 }
 {id=WavePlayer_24, properties {soundWaveAssetPtr=&/Spot_Building_Fixtures_Small_Creaks_01}}
 {id=WavePlayer_25, properties {soundWaveAssetPtr=&/Spot_Building_Fixtures_Small_Creaks_02}}
 {id=WavePlayer_26, properties {soundWaveAssetPtr=&/Spot_Building_Fixtures_Small_Creaks_03}}
 {id=WavePlayer_27, properties {soundWaveAssetPtr=&/Spot_Building_Fixtures_Small_Creaks_04}}
 {id=WavePlayer_28, properties {soundWaveAssetPtr=&/Spot_Building_Fixtures_Small_Creaks_05}}
 {id=WavePlayer_29, properties {soundWaveAssetPtr=&/Spot_Building_Fixtures_Small_Creaks_06}}
 {id=WavePlayer_30, properties {soundWaveAssetPtr=&/Spot_Building_Fixtures_Small_Creaks_07}}
 {id=WavePlayer_31, properties {soundWaveAssetPtr=&/Spot_Building_Fixtures_Small_Creaks_08}}
 {id=WavePlayer_32, properties {soundWaveAssetPtr=&/Spot_Building_Fixtures_Small_Creaks_09}}
 {id=WavePlayer_33, properties {soundWaveAssetPtr=&/Spot_Building_Fixtures_Small_Creaks_10}}
 {id=WavePlayer_34, properties {soundWaveAssetPtr=&/Spot_Building_Fixtures_Small_Creaks_11}}
 {id=WavePlayer_35, properties {soundWaveAssetPtr=&/Spot_Building_Fixtures_Small_Creaks_12}}
]
properties {
 attenuationSettings=/Game/Sounds/Attenuation/Fort_TileAmbient_OneShot_Wider_Att
 bDebug=true
 bOverrideConcurrency=true
 maxDistance=8000
 soundClassObject=&Fort_Ambient_Indoors_Athena
 volumeMultiplier=1
 attenuationOverrides {bSpatialize=false, stereoSpread=1147.9967, attenuationShapeExtents {x=6826.1992}}
 concurrencyOverrides {maxCount=2}
 concurrencySet [/AmbientAudioBase/Concurrency/AmbientConc_2Max_Loop_PreventNew_Default_Concurrency]
}
