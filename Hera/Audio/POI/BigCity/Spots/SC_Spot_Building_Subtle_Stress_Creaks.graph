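# Sound cue that plays subtle building stress creaks as an ambient spot sound for the BigCity POI.
# A Random node picks among twelve equally weighted creak waves and feeds a Modulator that slightly varies volume.
# It uses a shared wide one-shot tile-ambient attenuation asset and the Athena indoor ambient sound class.
# A concurrency override caps it at two simultaneous instances, and the cue's debug flag is left on.
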SoundCue /Hera/Audio/POI/BigCity/Spots/SC_Spot_Building_Subtle_Stress_Creaks
root=Modulator_0
nodes [
 {id=Modulator_0, children [Random_0], properties {volumeMax=1, volumeMin=0.9}}
 {
  id=Random_0
  children [
   WavePlayer_2
   WavePlayer_3
   WavePlayer_5
   WavePlayer_6
   WavePlayer_7
   WavePlayer_8
   WavePlayer_9
   WavePlayer_10
   WavePlayer_11
   WavePlayer_12
   WavePlayer_20
   WavePlayer_21
  ]
  properties {weights [1,1,1,1,1,1,1,1,1,1,1,1]}
 }
 {id=WavePlayer_10, properties {soundWaveAssetPtr=&/Spot_Building_Subtle_Stress_Creaks_08}}
 {id=WavePlayer_11, properties {soundWaveAssetPtr=&/Spot_Building_Subtle_Stress_Creaks_09}}
 {id=WavePlayer_12, properties {soundWaveAssetPtr=&/Spot_Building_Subtle_Stress_Creaks_10}}
 {id=WavePlayer_2, properties {soundWaveAssetPtr=&/Spot_Building_Subtle_Stress_Creaks_01}}
 {id=WavePlayer_20, properties {soundWaveAssetPtr=&/Spot_Building_Subtle_Stress_Creaks_11}}
 {id=WavePlayer_21, properties {soundWaveAssetPtr=&/Spot_Building_Subtle_Stress_Creaks_12}}
 {id=WavePlayer_3, properties {soundWaveAssetPtr=&/Spot_Building_Subtle_Stress_Creaks_02}}
 {id=WavePlayer_5, properties {soundWaveAssetPtr=&/Spot_Building_Subtle_Stress_Creaks_03}}
 {id=WavePlayer_6, properties {soundWaveAssetPtr=&/Spot_Building_Subtle_Stress_Creaks_04}}
 {id=WavePlayer_7, properties {soundWaveAssetPtr=&/Spot_Building_Subtle_Stress_Creaks_05}}
 {id=WavePlayer_8, properties {soundWaveAssetPtr=&/Spot_Building_Subtle_Stress_Creaks_06}}
 {id=WavePlayer_9, properties {soundWaveAssetPtr=&/Spot_Building_Subtle_Stress_Creaks_07}}
]
properties {
 attenuationSettings=/Game/Sounds/Attenuation/Fort_TileAmbient_OneShot_Wider_Att
 bDebug=true
 bOverrideConcurrency=true
 maxDistance=8000
 soundClassObject=&Fort_Ambient_Indoors_Athena
 volumeMultiplier=1
 attenuationOverrides {bSpatialize=false, stereoSpread=1147.9967, attenuationShapeExtents {x=6826.1992}}
 concurrencyOverrides {maxCount=2}
 concurrencySet [/AmbientAudioBase/Concurrency/AmbientConc_2Max_Loop_PreventNew_Default_Concurrency]
}
