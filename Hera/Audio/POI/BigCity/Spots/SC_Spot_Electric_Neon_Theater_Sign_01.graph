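# Sound cue for an outdoor ambient spot loop, an electric neon theater sign in the BigCity POI.
# A mixer layers looping neon sign and large transformer buzz waves.
# A random node picks, at equal weight, that mix either directly or through a pitch-varying modulator.
# Overrides set capsule attenuation with occlusion and listener focus, plus a small concurrency cap with distance-based volume scaling.
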SoundCue /Hera/Audio/POI/BigCity/Spots/SC_Spot_Electric_Neon_Theater_Sign_01
root=Random_1
nodes [
 {id=Mixer_0, children [WavePlayer_1,WavePlayer_0], properties {inputVolume [1,1]}}
 {id=Modulator_0, children [Mixer_0], properties {pitchMax=1.1}}
 {id=Random_1, children [Modulator_0,Mixer_0], properties {weights [1,1]}}
 {id=WavePlayer_0, properties {bLooping=true, soundWaveAssetPtr=/AmbientAudioContent/Waves/SpottedLoops/Spot_PowerPlant_Neon_Sign_Loop_01}}
 {id=WavePlayer_1, properties {bLooping=true, soundWaveAssetPtr=/AmbientAudioContent/Waves/SpottedLoops/Spot_Transformer_Large_Buzz_01}}
]
properties {
 bEnableSubmixSends=false
 bOverrideAttenuation=true
 bOverrideConcurrency=true
 maxDistance=1950
 soundClassObject=&Fort_Ambient_Outdoors_Athena
 volumeMultiplier=0.5
 attenuationOverrides {
  attenuationShape=Capsule
  bEnableListenerFocus=true
  bEnableOcclusion=true
  falloffDistance=1300
  focusVolumeAttenuation=1.1
  nonFocusVolumeAttenuation=0.6
  occlusionInterpolationTime=1
  occlusionLowPassFilterFrequency=750
  occlusionTraceChannel=ECC_GameTraceChannel11
  occlusionVolumeAttenuation=0.5
  reverbWetLevelMin=0.6905
  stereoSpread=100
  attenuationShapeExtents=<650 200 400>
 }
 concurrencyOverrides {maxCount=3, volumeScaleMode=Distance}
}
