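# Sound cue for an indoor ambient spot in the BigCity POI, playing the start/stop cycle of an integrated appliance.
# A Random node picks one of three equally weighted wave variations and feeds a Modulator that slightly varies volume.
# It uses a shared one-shot tile-ambient attenuation asset and the indoor ambient sound class.
# An enabled concurrency override caps simultaneous instances at two, and the debug flag is left on.
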SoundCue /Hera/Audio/POI/BigCity/Spots/SC_Spot_Integrated_Appliance_Start_Stop_Cycle
root=Modulator_0
nodes [
 {id=Modulator_0, children [Random_0], properties {volumeMax=1, volumeMin=0.9}}
 {id=Random_0, children [WavePlayer_2,WavePlayer_3,WavePlayer_5], properties {weights [1,1,1]}}
 {id=WavePlayer_2, properties {soundWaveAssetPtr=&/Spot_Integrated_Appliance_Start_Stop_Cycle_01}}
 {id=WavePlayer_3, properties {soundWaveAssetPtr=&/Spot_Integrated_Appliance_Start_Stop_Cycle_02}}
 {id=WavePlayer_5, properties {soundWaveAssetPtr=&/Spot_Integrated_Appliance_Start_Stop_Cycle_03}}
]
properties {
 attenuationSettings=/Game/Sounds/Attenuation/Fort_TileAmbient_OneShot_Wider_Att
 bDebug=true
 bOverrideConcurrency=true
 maxDistance=8000
 soundClassObject=&Fort_Ambient_Indoors_Athena
 volumeMultiplier=1
 attenuationOverrides {bSpatialize=false, stereoSpread=1147.9967, attenuationShapeExtents {x=6826.1992}}
 concurrencyOverrides {maxCount=2}
 concurrencySet [/AmbientAudioBase/Concurrency/AmbientConc_2Max_Loop_PreventNew_Default_Concurrency]
}
