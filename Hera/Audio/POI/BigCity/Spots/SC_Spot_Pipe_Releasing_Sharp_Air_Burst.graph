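# Ambient spot SoundCue for a pipe releasing a sharp air burst in the BigCity POI, using an indoor ambient sound class.
# A Mixer layers all eight burst waves at equal volume, so they start together rather than one being picked at random.
# A root Modulator slightly varies the volume of the mix.
# A shared wide one-shot attenuation asset handles falloff, and a concurrency override caps it at two instances.
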SoundCue /Hera/Audio/POI/BigCity/Spots/SC_Spot_Pipe_Releasing_Sharp_Air_Burst
root=Modulator_0
nodes [
 {
  id=Mixer_0
  children [WavePlayer_2, WavePlayer_3, WavePlayer_5, WavePlayer_11, WavePlayer_12, WavePlayer_13, WavePlayer_14, WavePlayer_15]
  properties {inputVolume [1,1,1,1,1,1,1,1]}
 }
 {id=Modulator_0, children [Mixer_0], properties {volumeMax=1, volumeMin=0.9}}
 {id=WavePlayer_11, properties {soundWaveAssetPtr=&/Spot_Pipe_Releasing_Sharp_Air_Burst_04}}
 {id=WavePlayer_12, properties {soundWaveAssetPtr=&/Spot_Pipe_Releasing_Sharp_Air_Burst_05}}
 {id=WavePlayer_13, properties {soundWaveAssetPtr=&/Spot_Pipe_Releasing_Sharp_Air_Burst_06}}
 {id=WavePlayer_14, properties {soundWaveAssetPtr=&/Spot_Pipe_Releasing_Sharp_Air_Burst_07}}
 {id=WavePlayer_15, properties {soundWaveAssetPtr=&/Spot_Pipe_Releasing_Sharp_Air_Burst_08}}
 {id=WavePlayer_2, properties {soundWaveAssetPtr=&/Spot_Pipe_Releasing_Sharp_Air_Burst_01}}
 {id=WavePlayer_3, properties {soundWaveAssetPtr=&/Spot_Pipe_Releasing_Sharp_Air_Burst_02}}
 {id=WavePlayer_5, properties {soundWaveAssetPtr=&/Spot_Pipe_Releasing_Sharp_Air_Burst_03}}
]
properties {
 attenuationSettings=/Game/Sounds/Attenuation/Fort_TileAmbient_OneShot_Wider_Att
 bDebug=true
 bOverrideConcurrency=true
 maxDistance=8000
 soundClassObject=&Fort_Ambient_Indoors_Athena
 volumeMultiplier=1
 attenuationOverrides {bSpatialize=false, stereoSpread=1147.9967, attenuationShapeExtents {x=6826.1992}}
 concurrencyOverrides {maxCount=2}
 concurrencySet [/AmbientAudioBase/Concurrency/AmbientConc_2Max_Loop_PreventNew_Default_Concurrency]
}
