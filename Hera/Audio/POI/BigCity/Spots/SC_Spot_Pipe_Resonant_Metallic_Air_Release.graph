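# Sound cue for a BigCity POI ambient spot effect: a resonant metallic air release from a pipe.
# A Random node picks among ten equally weighted wave variations and feeds a Modulator that slightly varies volume per play.
# It uses the indoor ambient sound class and a shared wide one-shot ambient attenuation asset.
# A concurrency override caps it at two simultaneous instances, and the debug flag is left on.
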SoundCue /Hera/Audio/POI/BigCity/Spots/SC_Spot_Pipe_Resonant_Metallic_Air_Release
root=Modulator_0
nodes [
 {id=Modulator_0, children [Random_0], properties {volumeMax=1, volumeMin=0.9}}
 {
  id=Random_0
  children [
   WavePlayer_2
   WavePlayer_3
   WavePlayer_5
   WavePlayer_13
   WavePlayer_14
   WavePlayer_15
   WavePlayer_16
   WavePlayer_17
   WavePlayer_18
   WavePlayer_19
  ]
  properties {weights [1,1,1,1,1,1,1,1,1,1]}
 }
 {id=WavePlayer_13, properties {soundWaveAssetPtr=&/Spot_Pipe_Resonant_Metallic_Air_Release_04}}
 {id=WavePlayer_14, properties {soundWaveAssetPtr=&/Spot_Pipe_Resonant_Metallic_Air_Release_05}}
 {id=WavePlayer_15, properties {soundWaveAssetPtr=&/Spot_Pipe_Resonant_Metallic_Air_Release_06}}
 {id=WavePlayer_16, properties {soundWaveAssetPtr=&/Spot_Pipe_Resonant_Metallic_Air_Release_07}}
 {id=WavePlayer_17, properties {soundWaveAssetPtr=&/Spot_Pipe_Resonant_Metallic_Air_Release_08}}
 {id=WavePlayer_18, properties {soundWaveAssetPtr=&/Spot_Pipe_Resonant_Metallic_Air_Release_09}}
 {id=WavePlayer_19, properties {soundWaveAssetPtr=&/Spot_Pipe_Resonant_Metallic_Air_Release_10}}
 {id=WavePlayer_2, properties {soundWaveAssetPtr=&/Spot_Pipe_Resonant_Metallic_Air_Release_01}}
 {id=WavePlayer_3, properties {soundWaveAssetPtr=&/Spot_Pipe_Resonant_Metallic_Air_Release_02}}
 {id=WavePlayer_5, properties {soundWaveAssetPtr=&/Spot_Pipe_Resonant_Metallic_Air_Release_03}}
]
properties {
 attenuationSettings=/Game/Sounds/Attenuation/Fort_TileAmbient_OneShot_Wider_Att
 bDebug=true
 bOverrideConcurrency=true
 maxDistance=8000
 soundClassObject=&Fort_Ambient_Indoors_Athena
 volumeMultiplier=1
 attenuationOverrides {bSpatialize=false, stereoSpread=1147.9967, attenuationShapeExtents {x=6826.1992}}
 concurrencyOverrides {maxCount=2}
 concurrencySet [/AmbientAudioBase/Concurrency/AmbientConc_2Max_Loop_PreventNew_Default_Concurrency]
}
